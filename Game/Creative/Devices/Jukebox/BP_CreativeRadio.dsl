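; Creative radio device that plays a selected song at itself or attached to registered instigator players.
; The server stamps song start time and replicates instigators, cleared on minigame state changes, plus volume, attenuation and visualization settings.
; Clients async-load the song, apply audibility rules and reuse existing components, starting newly attached player sounds at the elapsed offset.
; A hotfixable flag gates instigator-based audio, and AudioToMPC tracks sounds for visualization.
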(event EventBeginPlay
  ;[0.09665 1 0] EventBeginPlay
  (|CallParentFunction)
  ;;[1 0.787 0.276] Bind to Mini Game State Changed.
  (Minigame|BindEventtoMinigameStateChangedDelegate (Variables|BuildingPropPlaygroundMusicPlayer|GetMinigameLogicComponent) (EventDispatchers|CreateEvent :!event Custom|MinigameStateChanged))
  (Networking|CanExecuteCosmeticEvents
    (:True
      (CallFunction|SetAttenuationDistance)
      (CallFunction|UpdateCachedAttenuationSettings))))

(event EventEndPlay
  (params (EndPlayReason :type EEndPlayReason%20Enum :default "Destroyed"))
  ;[1 0.3935 0] EventEndPlay - Unbind from bindings.
  (Utilities|FlowControl|Sequence
    (:then_0
      (Minigame|UnbindEventfromMinigameStateChangedDelegate (Variables|BuildingPropPlaygroundMusicPlayer|GetMinigameLogicComponent) (EventDispatchers|CreateEvent :!event Custom|MinigameStateChanged)))
    (:then_1
      (|CallParentFunction EndPlayReason))))

(event EventOnWorldReady
  ;[0.04267 1 0] EventOnWorldReady
  (Visibility|SetVisibilityComponents (Variables|Default|GetVisibleInGameComponent) (Utilities|Array|MakeArray (Variables|Debug|GetStaticMeshComponent)) true))

(event OnPropertiesLoaded(ToyOptionsComponent)
  ;[0 1 0.06377] On Properties Loaded/Changed
  (CallFunction|PlaybackTypeUpgradePath)
  (CallFunction|OnPropertyChanged))

(event OnAnyPropertyChanged(ToyOptionsComponent)
  ;[0 1 0.06377] On Properties Loaded/Changed
  (CallFunction|OnPropertyChanged))

(event EventLoadThenPlaySong
  ;[0.005281 1 0 0.5] EventLoadThenPlaySong - The server will always load the song; however, clients may or may not depending on the conditions in PlayCurrentSongOnClient.
  ;;[1 0.9871 0.276] Hot-fixable option to turn off instigator location based radio. If false, this will return the radio to its previous functionality by removing the ability to spawn audio components on instigators.
  (if (ScalableFloat|GetValueasBool (Variables|Default|GetbAllowInstigatorLocationBasedAudio))
    (Networking|SwitchHasAuthority
      (:Authority
        (|CallParentFunction)
        (Variables|Default|SetTimeSongStarted (Utilities|Time|GetAudioTimeSeconds)))
      (:Remote
        ;;[1 0.9506 0.4219] Server doesn't get called while placing, so we need to call PlayCurrentSongOnClient to hear the radio when placing.
        (if (Math|Byte|Byte<=Byte (Variables|Default|GetCurrentMinigameState) (Utilities|Enum|LiteralenumEFortMinigameState "PreGame"))
          (CallFunction|PlayCurrentSongonClient true))))
    (else
      (|CallParentFunction))))

(event Custom|PlayCurrentSongOnClient
  (params (bShouldStopSound :type Boolean))
  (bind _outsong (CallFunction|GetCurrentSong))
  (bind _instigators_idx_to_ignore (Variables|Default|GetInstigators%20Idx%20to%20Ignore))
  (bind _sound_idxtoignore (Variables|Default|GetSound%20IdxToIgnore))
  ;[0.02611 1 0.06393 0.5] PlayCurrentSoundOnClient - Stop all sounds if flagged, load the current song, and spawn the audio component either at the radio placement location or instigator(s) location.
  ;;[0.9887 1 0.2396] Prevents audio components from being spawned on instigators if bOverrideInstigatorLocationBasedRadio is true.
  (if (ScalableFloat|GetValueasBool (Variables|Default|GetbAllowInstigatorLocationBasedAudio))
    (if bShouldStopSound
      ;;[0.3099 0.08274 1] Stop all sounds
      (CallFunction|StopNonExcludedSounds (Utilities|Array|MakeArray -1))
      (else
        _))
    ;;[1 0.8231 0.2708] Load the current song. ShouldPlaySong must be called after loading due to racing conditions that can occur.
    (Networking|SwitchHasAuthority
      (:Remote
        (if (Utilities|NotEqual(SoftObjectReference) (Variables|Default|GetPendingLoadSong) _outsong)
          (bind _output_get (Variables|Default|SetPendingLoadSong _outsong))
          (bind _object (Utilities|AsyncLoadAsset _output_get)
            (:Completed
              (Utilities|IsValid _object
                (:Is%20Valid
                  (if (Utilities|Equal(SoftObjectReference) (Utilities|ResolveSoftReference _object) (Variables|Default|GetPendingLoadSong))
                    (Variables|Default|SetPendingLoadSong 0) ; Clear the song to unblock subsequent load attempts
                    (bind _bshouldplaysong (CallFunction|ShouldPlaySong))
                    (if _bshouldplaysong
                      (switch Utilities|FlowControl|Switch|SwitchonE_PlayAtLocationCreativeRadio (Utilities|Enum|BytetoEnumE_PlayAtLocationCreativeRadio (Math|Conversions|ToByte(Integer) (Variables|User_Options|GetPlayatLocationIndex)))
                        (:NewEnumerator0
                          ;;[0.562 0.4262 1] Spawn the sound at the device's location unlesss there is one already.
                          (bind _bradiohassoundcomponent (CallFunction|DoesRadioAlreadyHaveSoundComponent))
                          (if (not _bradiohassoundcomponent)
                            (CallFunction|SpawnSoundAttachedtoComponent _object (Variables|Debug|GetStaticMeshComponent) true)))
                        (:NewEnumerator1
                          ;;[0.2171 1 0.1082] Spawn the sound at the instigator's location. There may be more than one instigator.
                          ;;;[0.9514 1 0.4375] Before spawning any audio components, determine indicies from the SpawnedSounds and Instigators array that should be ignored. This will prevent unnecessarily spawning and destroying audio components.
                          (CallFunction|CreateIdxtoIgnore _instigators_idx_to_ignore _sound_idxtoignore)
                          ;;;[0.6049 0.3695 1] If sounds aren't on the exclusion list, stop them.
                          (CallFunction|StopNonExcludedSounds _sound_idxtoignore)
                          ;;;[0.7518 0.404 1] If the instigator isn't on the ignore list, spawn a sound attached to them.
                          (for (_array_element _array_index) (Variables|Default|GetInstigators)
                            (if (not (Utilities|Array|ContainsItem _instigators_idx_to_ignore _array_index))
                              (Utilities|IsValid _array_element
                                (:Is%20Valid
                                  (bind _pawnprivate (Variables|/Script/Engine.PlayerState|GetPawnPrivate _array_element)
                                    (:then
                                      (CallFunction|SpawnSoundAttachedtoComponent _object (Variables|/Script/Engine.Actor|GetRootComponent _pawnprivate)))
                                    (:else))))))))
                      (else
                        (CallFunction|StopNonExcludedSounds (Utilities|Array|MakeArray -1))))))))))))))

(event Custom|OnInstigatorsUpdated
  ;[0 1 0.1279 0.5] On Instigators Updated - Play the current song on the client.
  (CallFunction|PlayCurrentSongonClient false))

(event Custom|Event%20On%20Receive%20Play
  (params (Instigator :type Controller%20Object%20Reference))
  ;[0 1 0.101] Gameplay Functions
  (CallFunction|Play))

(event Custom|Event%20On%20Receive%20Stop
  (params (Instigator :type Controller%20Object%20Reference))
  ;[0 1 0.101] Gameplay Functions
  (CallFunction|Stop))

(event Custom|Event%20On%20Receive%20Register%20Player
  (params (Instigator :type Controller%20Object%20Reference))
  (bind _playerstate (Variables|/Script/Engine.Controller|GetPlayerState Instigator))
  (bind _instigators (Variables|Default|GetInstigators))
  ;[0 1 0.101] Gameplay Functions
  ;;[1 0.8827 0.3438] Update Instigators array when instigators are Resgistered/Unregistered.
  (Networking|SwitchHasAuthority
    (:Authority
      (Utilities|IsValid _playerstate
        (:Is%20Valid
          (bind _returnvalue (Utilities|Array|AddUnique _instigators _playerstate))
          (if (!= _returnvalue -1)
            ;;;[1 0.8269 0.3854] OnReps will not get called unless set after adding or removing.
            (Variables|Default|SetInstigators _instigators)))))))

(event Custom|Event%20On%20Receive%20Unregister%20Player
  (params (Instigator :type Controller%20Object%20Reference))
  (bind _playerstate (Variables|/Script/Engine.Controller|GetPlayerState Instigator))
  (bind _instigators (Variables|Default|GetInstigators))
  ;[0 1 0.101] Gameplay Functions
  ;;[1 0.8827 0.3438] Update Instigators array when instigators are Resgistered/Unregistered.
  (Networking|SwitchHasAuthority
    (:Authority
      (Utilities|IsValid _playerstate
        (:Is%20Valid
          (bind _returnvalue (Utilities|Array|RemoveItem _instigators _playerstate))
          (if _returnvalue
            ;;;[1 0.8269 0.3854] OnReps will not get called unless set after adding or removing.
            (Variables|Default|SetInstigators _instigators)))))))

(event Custom|Event%20On%20Receive%20Unregister%20All%20Players
  (params (Instigator :type Controller%20Object%20Reference))
  (bind _instigators (Variables|Default|GetInstigators))
  ;[0 1 0.101] Gameplay Functions
  ;;[1 0.8827 0.3438] Update Instigators array when instigators are Resgistered/Unregistered.
  (Networking|SwitchHasAuthority
    (:Authority
      (Utilities|Array|Clear _instigators)
      ;;;[1 0.8269 0.3854] OnReps will not get called unless set after adding or removing.
      (Variables|Default|SetInstigators _instigators))))

(event Custom|MinigameStateChanged
  (params (Minigame :type Fort%20Minigame%20Object%20Reference) (NewMinigameState :type EFortMinigameState%20Enum :default "PreGame"))
  (bind _instigators (Variables|Default|GetInstigators))
  ;[0.06763 1 0] Minigame Stated Changed
  (Variables|Default|SetCurrentMinigameState NewMinigameState)
  ;;[1 0.9 0.3646] Clear instigators when game state changes.
  (Networking|SwitchHasAuthority
    (:Authority
      (Utilities|Array|Clear _instigators)
      (Variables|Default|SetInstigators _instigators))))

(event OnAddedtoMinigame(FortMinigameProgress)
  (params (Minigame :type Fort%20Minigame%20Object%20Reference))
  (Networking|SwitchHasAuthority
    (:Authority
      (CallFunction|CheckMinigameStateonPropertyChanged))))

(fn BlueprintCanInteract
  (params (InteractingPawn :type Fort%20Pawn%20Object%20Reference) (InteractionBeingAttempted :type EInteractionBeingAttempted%20Enum :default "FirstInteraction") (InteractionType :type TInteractionType%20Enum :default "IT_NoInteraction"))
  (bind _returnvalue (Pawn|GetController InteractingPawn))
  (bind _fortminigameprogress (Variables|Default|GetFortMinigameProgress))
  (bind _currentminigame (Variables|/Script/FortniteGame.FortMinigameLogicComponent|GetCurrentMinigame _fortminigameprogress))
  (LocalVariables|SetInteractSuccess false)
  (Utilities|FlowControl|Sequence
    (:then_0
      (Utilities|IsValid (Variables|/Script/FortniteGame.FortActorOptionsComponent|GetInternal_PlayerOptions (Variables|Default|GetToyOptionsComponent))
        (:Is%20Valid
          (Utilities|IsValid InteractingPawn
            (:Is%20Valid
              (Utilities|IsValid _returnvalue
                (:Is%20Valid
                  (bind _asfort_player_controller_athena (Utilities|Casting|CastToFortPlayerControllerAthena _returnvalue))
                  (if (Creative|CanPlayerCreateinCurrentExperience _asfort_player_controller_athena)
                    (Utilities|IsValid _currentminigame
                      (:Is%20Valid
                        (LocalVariables|SetInteractSuccess (not (Minigame|HasStarted _currentminigame :!class /Script/FortniteGame.FortMinigame))))
                      (:Is%20Not%20Valid
                        (LocalVariables|SetInteractSuccess true)))))))))))
    (:then_1
      (return (LocalVariables|GetInteractSuccess)))))

(fn BlueprintGetInteractionString
  (params (InteractingPawn :type Fort%20Pawn%20Object%20Reference) (InteractionBeingAttempted :type EInteractionBeingAttempted%20Enum :default "FirstInteraction"))
  (return "NSLOCTEXT(\"[F7C22CCD4C685EFD4763FEA8B4E07B78]\", \"6F39C9CD4BC9425D3C4752AD450470B6\", \"Customize\")"))

(fn CheckMinigameStateOnPropertyChanged
  (bind _currentminigame (Variables|/Script/FortniteGame.FortMinigameLogicComponent|GetCurrentMinigame (Variables|Default|GetFortMinigameProgress))
    (:then
      (CallFunction|OnMinigameStateChanged 0 (Variables|/Script/FortniteGame.FortMinigame|GetCurrentState _currentminigame)))
    (:else)))

(fn CreateIdxToIgnore
  (params (OutInstigatorsIdxToIgnore :type Array%20of%20Integers) (OutSoundIdxtoSave :type Array%20of%20Integers))
  (bind _instigators (Variables|Default|GetInstigators))
  ;[1 0.919 0.3125] Since each player state can only have one spawned sound associated with it, go through the SpawnedSounds and see if any player states are already represented. If so add the index of the sound and instigator player state for later comparrison. This will make us avoid destroying and re-spawning audio components we already have.
  (Utilities|Array|Clear OutInstigatorsIdxToIgnore)
  (Utilities|Array|Clear OutSoundIdxtoSave)
  (for (_array_element _array_index) (Variables|Default|GetSpawnedSounds)
    (bind _asfort_player_pawn (Utilities|Casting|CastToFortPlayerPawn (Components|GetOwner _array_element)))
    (if (!= (Utilities|Array|FindItem _instigators (Variables|/Script/Engine.Pawn|GetPlayerState _asfort_player_pawn)) -1)
      (Utilities|Array|AddUnique OutInstigatorsIdxToIgnore (Utilities|Array|FindItem _instigators (Variables|/Script/Engine.Pawn|GetPlayerState _asfort_player_pawn)))
      (Utilities|Array|AddUnique OutSoundIdxtoSave _array_index)))
  (return))

(fn DoesRadioAlreadyHaveSoundComponent
  ;[0.05719 1 0.2929 0.5] Check if any of the spawned sound components are owned by this device. Sound components can either be owned by the device or players.
  (for _array_element (Variables|Default|GetSpawnedSounds)
    (Utilities|IsValid _array_element
      (:Is%20Valid
        (if (Utilities|Equal(Object) (Components|GetOwner _array_element) self)
          (return true)))))
  (return false))

(fn GetCurrentSong
  ;[0.1232 1 0.3635] Get the current song from the MusicArray.
  (return (select (Utilities|Array|IsValidIndex (Variables|BuildingPropMusicPlayer|GetMusicArray) (Utilities|Struct|BreakMusicPlayerData (Variables|BuildingPropMusicPlayer|GetPlayingSongData))) (Utilities|Array|Get(acopy) (Variables|BuildingPropMusicPlayer|GetMusicArray) (Utilities|Struct|BreakMusicPlayerData (Variables|BuildingPropMusicPlayer|GetPlayingSongData))) 0)))

(fn IsTargetableByLinearMover
  (return true))

(fn OnPropertyChanged
  (Networking|SwitchHasAuthority
    (:Authority
      (Utilities|FlowControl|Sequence
        (:then_0
          (if (!= (Variables|Default|GetVolumeOverride) (Variables|Default|GetCachedVolumeMultiplier))
            (Variables|Default|SetCachedVolumeMultiplier (Variables|Default|GetVolumeOverride))))
        (:then_1
          (CallFunction|SetAttenuationDistance))
        (:then_2
          (if (Math|Boolean|NotEqual(Boolean) (Variables|Default|GetShouldDriveVisualization) (Variables|Default|GetCachedShouldDriveVisualization))
            (Variables|Default|SetCachedShouldDriveVisualization (Variables|Default|GetShouldDriveVisualization))))
        (:then_3
          (CallFunction|StartSongforOptionsChange)
          (CallFunction|CheckMinigameStateonPropertyChanged))
        (:then_4
          (CallFunction|Resave))
        (:then_5
          (Visibility|SetVisiblityDuringPhase (Variables|Default|GetVisibleInGameComponent) (select (Variables|Default|GetIsMeshVisibleDuringMinigame) "NewEnumerator3" "NewEnumerator1")))))
    (:Remote
      (CallFunction|PlayCurrentSongonClient true))))

(fn OnRep_CachedAttenuationSetting
  (CallFunction|UpdateCachedAttenuationSettings))

(fn OnRep_CachedShouldDriveVisualization
  (Networking|CanExecuteCosmeticEvents
    (:True
      (Components|Activation|SetActive (Variables|Default|GetCreative_Radio_AudioToMPC) (Variables|Default|GetCachedShouldDriveVisualization) false))))

(fn OnRep_CachedVolumeMultiplier
  (bind _cachedvolumemultiplier (Variables|Default|GetCachedVolumeMultiplier))
  ;[1 0.8713 0.5104] Set the volume multiplier for all audio components.
  (Networking|CanExecuteCosmeticEvents
    (:True
      (Audio|Components|Audio|SetVolumeMultiplier (Variables|BuildingPropMusicPlayer|GetSongAudioComponent) _cachedvolumemultiplier :!class /Script/Engine.AudioComponent)
      (for _array_element (Variables|Default|GetSpawnedSounds)
        (Audio|Components|Audio|SetVolumeMultiplier _array_element _cachedvolumemultiplier :!class /Script/Engine.AudioComponent)))))

(fn OnRep_Instigators
  ;[0 1 0.1927] Call when when instigators get added or removed.
  (CallFunction|OnInstigatorsUpdated))

(fn OnRep_TimeSongStarted
  ;[0.2803 1 0] Play the current song on the client. Called whenever a song is loaded on the server.
  (CallFunction|PlayCurrentSongonClient true))

(fn PlaybackTypeUpgradePath
  (bind _toyoptionscomponent (Variables|Default|GetToyOptionsComponent))
  (bind _currentoptions (OptionData|GetPropertyOverrides _toyoptionscomponent))
  (bind (_value _returnvalue) (Utilities|Map|Find _currentoptions "PlaybackType"))
  (bind _toyoptionscomponent_2 (Variables|Default|GetToyOptionsComponent))
  (if _returnvalue
    (OptionData|SetPropertyOverrides _toyoptionscomponent_2 (Utilities|Map|MakeMap "PlayDuringPregame" (select (Utilities|String|Equal,CaseInsensitive(String) _value "1") "Yes" "No") "PlayDuringWarmup" (select (Utilities|String|Equal,CaseInsensitive(String) _value "1") "Yes" "No") "PlayDuringGameplay" (select (Utilities|String|Equal,CaseInsensitive(String) _value "1") "Yes" "No") "PlayDuringRoundEnd" (select (Utilities|String|Equal,CaseInsensitive(String) _value "1") "Yes" "No") "PlayDuringGameEnd" (select (Utilities|String|Equal,CaseInsensitive(String) _value "1") "Yes" "No")))
    (OptionData|RemovePropertyOverrides _toyoptionscomponent_2 (Utilities|Array|MakeArray "PlaybackType"))))

(fn SetAttenuationDistance
  (if (!= (Variables|Default|GetRadioAttenuationDistance) (Variables|Default|GetCachedAttenuationSetting))
    (Variables|Default|SetCachedAttenuationSetting (Variables|Default|GetRadioAttenuationDistance))))

(fn ShouldInvisiblePropsBeAllowed
  (return false))

(fn ShouldPlaySong
  (bind _returnvalue (Fortnite|GetFirstLocalFortPlayerController true))
  (bind _playerstate (Variables|/Script/Engine.Controller|GetPlayerState _returnvalue))
  (bind _instigators (Variables|Default|GetInstigators))
  (bind _returnvalue_2 (Utilities|Array|ContainsItem _instigators _playerstate))
  ;[0 1 0.05694 0.5] Determine if we should be hearing anything at all by comparing the CanBeHeardbyIndex to the instigators.
  (Utilities|IsValid _returnvalue
    (:Is%20Valid
      (switch Utilities|FlowControl|Switch|SwitchonE_CanBeHeardByCreativeRadio (Utilities|Enum|BytetoEnumE_CanBeHeardByCreativeRadio (Math|Conversions|ToByte(Integer) (Variables|UserOptions|GetCanBeHeardByIndex)))
        (:NewEnumerator0
          (return true))
        (:NewEnumerator1
          (return _returnvalue_2))
        (:NewEnumerator2
          (return (not _returnvalue_2)))))))

(fn SpawnSoundAttachedToComponent
  (params (InSound :type Object%20Reference) (InComponent :type Scene%20Component%20Object%20Reference) (PlayOnDevice :type Boolean))
  (bind _songaudiocomponent (Variables|BuildingPropMusicPlayer|GetSongAudioComponent))
  (bind _localaudiocomponentreference (LocalVariables|GetLocalAudioComponentReference))
  ;[0.1845 1 0.04048] Spawn the sound attached to InComponent and configure.
  (bind _assound_base (Utilities|Casting|CastToSoundBase InSound))
  ;[1 1 1] Play The Built in Component if not attaching to a player, so it will work with the Audio Visualizer Gallery
  (if PlayOnDevice
    (Variables|/Script/Engine.AudioComponent|SetSound _assound_base _songaudiocomponent)
    (Audio|Components|Audio|Play _songaudiocomponent 0.0)
    (LocalVariables|SetLocalAudioComponentReference _songaudiocomponent)
    (else
      (bind _returnvalue (Audio|SpawnSoundAttached :Sound _assound_base :AttachToComponent InComponent :AttachPointName "None" :Location (Transformation|GetWorldLocation InComponent) :LocationType "SnapToTarget" :bStopWhenAttachedToDestroyed true :VolumeMultiplier (Variables|Default|GetCachedVolumeMultiplier) :PitchMultiplier (Variables|/Script/Engine.AudioComponent|GetPitchMultiplier (Variables|BuildingPropMusicPlayer|GetSongAudioComponent)) :StartTime (- (Utilities|Time|GetAudioTimeSeconds) (Variables|Default|GetTimeSongStarted)) :bAutoDestroy true))
      (bind _output_get (LocalVariables|SetLocalAudioComponentReference _returnvalue))
      ;;[1 0.9495 0.08993 1 12] Add the spawned component to the AudioToMPC component so the audio played by this radio can be visualized when attached to the player
      (AudioToMPC|AddTrackedAudioComponent (Variables|Default|GetCreative_Radio_AudioToMPC) _output_get)))
  (Utilities|IsValid (LocalVariables|GetLocalAudioComponentReference)
    (:Is%20Valid
      (Utilities|Array|Add (Variables|Default|GetSpawnedSounds) _localaudiocomponentreference)
      ;[1 0.9638 0.2314] Setup attenuation.
      (bind _structout (Utilities|Struct|SetmembersinSoundAttenuationSettings (Variables|/Script/Engine.AudioComponent|GetAttenuationOverrides _localaudiocomponentreference) (< (Variables|Default|GetCachedAttenuationSetting) (Variables|Config|GetMaxAttenuationDistance)) (* (Variables|Default|GetCachedAttenuationSetting) (Variables|Config|GetGridSize))))
      (Audio|Components|Audio|AdjustAttenuation _localaudiocomponentreference _structout))))

(fn StopNonExcludedSounds
  (params (InExcludedSoundIdx :type Array%20of%20Integers))
  ;[0.1939 1 0.0599] Stop, destroy, and remove sounds from the SpawnedSounds array if they're not contined in InExcludedSoundIdx.
  (bind (_arrayindex _arrayelement) (Utilities|Array|ReverseforEachLoop (Variables|Default|GetSpawnedSounds))
    (:LoopBody
      (Utilities|IsValid _arrayelement
        (:Is%20Valid
          (if (not (Utilities|Array|ContainsItem InExcludedSoundIdx _arrayindex))
            (Audio|Components|Audio|Stop _arrayelement)
            (if (Utilities|NotEqual(Object) _arrayelement (Variables|BuildingPropMusicPlayer|GetSongAudioComponent))
              ;;[0.9692 1 0.152 1 16] Only remove the component from the AudioToMPC component and destroy it if it is not attached to the device
              (AudioToMPC|RemoveTrackedAudioComponent (Variables|Default|GetCreative_Radio_AudioToMPC) _arrayelement)
              (Components|DestroyComponent _arrayelement)
              (else
                _))
            (Utilities|Array|RemoveIndex (Variables|Default|GetSpawnedSounds) _arrayindex)))))))

(fn UpdateCachedAttenuationSettings
  (bind _cachedattenuationsetting (Variables|Default|GetCachedAttenuationSetting))
  (bind _returnvalue (< _cachedattenuationsetting 1e+03))
  (bind _cachedattenuationsetting_2 (Variables|Default|GetCachedAttenuationSetting))
  (bind _returnvalue_2 (* _cachedattenuationsetting_2 512.0))
  (Networking|CanExecuteCosmeticEvents
    (:True
      (bind _structout (Utilities|Struct|SetmembersinSoundAttenuationSettings (Variables|/Script/Engine.AudioComponent|GetAttenuationOverrides (Variables|BuildingPropMusicPlayer|GetSongAudioComponent)) _returnvalue _returnvalue_2))
      (Audio|Components|Audio|AdjustAttenuation (Variables|BuildingPropMusicPlayer|GetSongAudioComponent) _structout)
      (for _array_element (Variables|Default|GetSpawnedSounds)
        (bind _structout_2 (Utilities|Struct|SetmembersinSoundAttenuationSettings (Variables|/Script/Engine.AudioComponent|GetAttenuationOverrides _array_element) _returnvalue _returnvalue_2))
        (Audio|Components|Audio|AdjustAttenuation _array_element _structout_2)))))

(fn ConstructionScript)
