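; Creative_Radio_AudioToMPC is a Blueprint stub for the Creative Jukebox device, named for passing radio audio data to a material parameter collection.
; It declares an OnGatheredFFTData function that takes an array of FFT channel data and a by-reference average amplitude.
; BeginPlay, Tick and that function show no logic, so no analysis or parameter write appears here.
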(event EventBeginPlay)

(event EventTick
  (params (DeltaSeconds :type Float%20%28single-precision%29)))

(fn OnGatheredFFTData
  (params (FFTData :type Array%20of%20Channel%20Data%20Structures) (OutAmplitudeAverage :type Float%20%28single-precision%29%20%28by%20ref%29))
  (return))
